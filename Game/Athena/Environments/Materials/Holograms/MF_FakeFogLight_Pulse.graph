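# Material function for a pulsing fake fog-light sphere, outputting emissive, opacity and world position offset.
# Opacity blends scene-depth surface lighting, shaded by GBuffer normals, with ray-traced sphere fog falloff.
# Two position-offset sines add pulsing, alongside time-of-day scaling and depth and camera fades.
# A static switch sets when vertices collapse to the camera, based on sphere radius and fade distance; forward and mobile always collapse.
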MaterialFunction /Game/Athena/Environments/Materials/Holograms/MF_FakeFogLight_Pulse
graph {nodes [
  n1 = Distance(A=ObjectPositionWS(), B=CameraPositionWS())
  n2 = ScalarParameter() {defaultValue=15000, parameterName=CameraFadeOffset}
  n3 = Clamp(None=Divide(A=Subtract(A=@n1, B=@n2), B=ScalarParameter() {defaultValue=10000, parameterName=CameraFadeLength}))
  n4 = Floor(None=@n3)
  n5 = Distance(A=ObjectPositionWS(), B=CameraPositionWS())
  n6 = ScalarParameter() {defaultValue=3579.1655, parameterName=SphereRadius}
  n7 = Ceil(None=Clamp(None=Subtract(A=@n5, B=@n6)))
  n8 = OneMinus(None=@n7)
  n9 = FunctionInput(Preview=StaticBool()) {bUsePreviewValueAsDefault=true, inputType=FunctionInput_StaticBool}
  n10 = StaticSwitch(True=Clamp(None=Add(A=@n4, B=@n8)), False=@n7, Value=@n9)
  n11 = Ceil()
  Add(A=@n10, B=OneMinus(None=Saturate(None=@n11)))
  n12 = Multiply(A=ParticleColor().RGB, B=VectorParameter() {parameterName=Color, defaultValue=<rgba 1 0.168 0.0312 1>}.RGB)
  n13 = ScalarParameter() {defaultValue=5, parameterName=Brightness}
  n14 = ShadingPathSwitch(Default=@n13, Deferred=@n13, Forward=Multiply(A=@n13) {constB=0.35})
  n15 = MaterialFunctionCall(
   Value_Morning=ScalarParameter() {defaultValue=1, parameterName="Value_Morning (S)"}
   Value_Day=ScalarParameter() {defaultValue=1, parameterName="Value_Day (S)"}
   Value_Evening=ScalarParameter() {defaultValue=1, parameterName="Value_Evening (S)"}
   Value_Night=ScalarParameter() {defaultValue=1, parameterName="Value_Night (S)"}
  ) {materialFunction=&Scalar_Time_of_Day}
  n16 = ParticleColor()
  n17 = MaterialFunctionCall() {materialFunction=&WorldPositionBehindTranslucency}
  n18 = CameraPositionWS()
  n19 = Clamp(None=Divide(A=Distance(A=@n17.Result, B=@n18), B=ScalarParameter() {defaultValue=2000, parameterName=NearFadeDistance}))
  n20 = ScalarParameter() {defaultValue=1, parameterName=LowEndOpacity}
  n21 = StaticSwitchParameter(True=ShadingPathSwitch(Default=@n19, Forward=@n20, Mobile=@n20), False=@n19) {parameterName=UseLowEndSettings?}
  n22 = Distance(
   A=MaterialFunctionCall() {materialFunction=&WorldPositionBehindTranslucency}.Result
   B=ObjectPositionWS()
  )
  n23 = ScalarParameter() {defaultValue=2706.8237, parameterName=LightRadius}
  n24 = OneMinus(None=Clamp(None=Divide(A=@n22, B=@n23)))
  n25 = ScalarParameter() {defaultValue=1, parameterName=LowEndOpacity}
  n26 = StaticSwitchParameter(True=ShadingPathSwitch(Default=@n24, Forward=@n25, Mobile=@n25), False=@n24) {parameterName=UseLowEndSettings?}
  n27 = SceneTexture() {sceneTextureId=PPI_WorldNormal}
  n28 = Subtract(
   A=ObjectPositionWS()
   B=MaterialFunctionCall() {materialFunction=&WorldPositionBehindTranslucency}.Result
  )
  n29 = LinearInterpolate(
   A=ScalarParameter() {defaultValue=1, parameterName=MinLightTop}
   Alpha=DotProduct(A=ComponentMask(None=@n27.Color) {b=true, g=true, r=true}, B=Normalize(VectorInput=@n28))
  ) {constA=1}
  n30 = Constant() {r=1}
  n31 = ShadingPathSwitch(Default=@n29, Deferred=@n29, Forward=@n30, Mobile=@n30)
  n32 = CameraPositionWS()
  n33 = Subtract(A=WorldPosition().XYZ, B=@n32) {constB=-1}
  n34 = OneMinus(None=MaterialFunctionCall!(
    Sphere Radius=Constant() {r=1}
    Sphere Falloff=ScalarParameter() {defaultValue=2706.8237, parameterName=LightRadius}
    Unit Ray Direction=Normalize(VectorInput=@n33)
    Sphere Center=ObjectPositionWS()
    Ray Origin=@n32
   ) {materialFunction=/Game/Athena/Environments/MaterialFunctions/RayTraceSphereFalloffMinClamp}.Clamped Falloff)
  n35 = ScalarParameter() {defaultValue=2, parameterName=FalloffFog}
  n36 = Clamp(None=Add(A=Multiply(A=@n26, B=@n31), B=Power(Base=@n34, Exp=@n35)))
  n37 = Multiply(A=VertexInterpolator(VS=OneMinus(None=@n3)).PS, B=@n36)
  n38 = StaticSwitch(
   True=DepthFade(Opacity=Multiply(A=@n21, B=@n37), FadeDistance=ScalarParameter() {defaultValue=2000, parameterName=DepthFadeDistance})
   False=Multiply(A=@n21, B=@n36)
   Value=@n9
  )
  n39 = Time()
  n40 = Frac(None=ComponentMask(None=ObjectPositionWS()) {r=true})
  n41 = MaterialFunctionCall!(
   Input=Sine(None=Add(A=@n39, B=@n40)) {period=6}
   Input Low=Constant()
   Input High=Constant() {r=1}
   Target Low=Constant() {r=0.7}
   Target Height=Constant() {r=1}
  ) {materialFunction=&RemapValueRange}
  n42 = Time()
  n43 = Frac(None=ComponentMask(None=ObjectPositionWS()) {r=true})
  n44 = MaterialFunctionCall(
   Input=Sine(None=Add(A=@n42, B=@n43)) {period=2.53}
   Input Low=Constant()
   Input High=Constant() {r=1}
   Target Low=Constant() {r=0.8}
   Target Height=Constant() {r=1}
  ) {materialFunction=&RemapValueRange}
  n45 = Multiply(
   A=Multiply(A=Multiply(A=@n16.A, B=@n38), B=Multiply(A=@n41.Result, B=@n44.Result))
   B=MaterialFunctionCall(
    Fade Length=ScalarParameter() {defaultValue=2500, parameterName=FadeWhenClose}
    Fade Offset=ScalarParameter() {defaultValue=2500, parameterName=FadeWhenCloseOffset}
   ) {materialFunction=&CameraDepthFade}.Result
  )
  n46 = Subtract(A=CameraPositionWS(), B=WorldPosition().XYZ)
  n47 = Multiply(A=@n46, B=Saturate(None=@n10))
  n48 = FeatureLevelSwitch(Default=@n47, ES2=@n46, ES3_1=@n46, SM4_REMOVED=@n47)
  FunctionOutput(None=MakeMaterialAttributes!(
    BaseColor
    Metallic
    Specular
    Roughness
    Anisotropy
    EmissiveColor=Multiply(A=@n12, B=@n14)
    Opacity=Multiply(A=@n15.Result, B=@n45)
    OpacityMask
    Normal
    Tangent
    WorldPositionOffset=ShadingPathSwitch(Default=@n48, Deferred=@n48, Forward=@n46, Mobile=@n46)
    SubsurfaceColor
    ClearCoat
    ClearCoatRoughness
    AmbientOcclusion
    Refraction
    CustomizedUVs_0
    CustomizedUVs_1
    CustomizedUVs_2
    CustomizedUVs_3
    CustomizedUVs_4
    CustomizedUVs_5
    CustomizedUVs_6
    CustomizedUVs_7
    PixelDepthOffset
    ShadingModel
    Displacement
   ))
 ]}
