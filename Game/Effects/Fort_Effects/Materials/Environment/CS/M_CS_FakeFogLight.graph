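# M_CS_FakeFogLight is an unlit, additive translucent environment material for a fake fog light effect.
# Its only node is a call to the MF_FakeFogLight_Pulse function from the Holograms materials, which holds all of the shading logic.
# It uses material attributes, disables depth testing and translucency vertex fog, and renders in the before-depth-of-field translucency pass.
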Material /Game/Effects/Fort_Effects/Materials/Environment/CS/M_CS_FakeFogLight
graph {
 properties {
  bDisableDepthTest=true
  bUseMaterialAttributes=true
  bUseTranslucencyVertexFog=false
  blendMode=BLEND_Additive
  pixelDepthOffsetMode=PDOM_Legacy
  shadingModel=MSM_Unlit
  translucencyPass=MTP_BeforeDOF
  usedShadingModels=Unlit
 }
 nodes [MaterialFunctionCall!(In) {materialFunction=/Game/Athena/Environments/Materials/Holograms/MF_FakeFogLight_Pulse}]
}
