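# Unlit additive material for a glow sphere effect, flagged for Cascade and Niagara mesh particles.
# A Color parameter, hue-shifted by vertex red times HueShift, is multiplied by a tunable Fresnel.
# Depth fade and switchable near-camera fade soften the result, but low quality skips both.
# Static switches derive a separate opacity-style value from the desaturated glow, particle alpha or a constant one.
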Material /Game/Effects/Fort_Effects/Materials/Environment/CS/M_Mother_Crash_GlowSphere
graph {
 properties {
  bUsedWithMeshParticles=true
  bUsedWithNiagaraMeshParticles=true
  blendMode=BLEND_Additive
  pixelDepthOffsetMode=PDOM_Legacy
  shadingModel=MSM_Unlit
  usedShadingModels=Unlit
 }
 nodes [
  n1 = VertexColor()
  n2 = ScalarParameter() {parameterName=HueShift}
  n3 = Multiply(
   A=MaterialFunctionCall(Hue Shift Percentage=Multiply(A=@n1.R, B=@n2), Texture=VectorParameter() {parameterName=Color, defaultValue=<rgba 0 0 0 1>}.RGB) {materialFunction=&HueShift}.Result
   B=MaterialFunctionCall(
    Brightness=ScalarParameter() {defaultValue=1, parameterName=Fresnel Brightness}
    Exponent=ScalarParameter() {defaultValue=10, parameterName=Fresnel Exponent}
   ) {materialFunction=&MF_Fresnel_FX}.Fresnel
  )
  n4 = ScalarParameter() {defaultValue=500, parameterName=CameraFadeDistance}
  n5 = Multiply(
   A=DepthFade(Opacity=@n3, FadeDistance=ScalarParameter() {defaultValue=500, parameterName=Depth Fade Distance})
   B=StaticSwitchParameter(
    True=MaterialFunctionCall(Distance=@n4) {materialFunction=&Near_Camera_Fade}.Result
    False=Constant() {r=1}
   ) {parameterName=UseCameraFade}
  )
  Reroute(None=QualitySwitch(Default=@n5, Low=@n3, High=@n5, Medium=@n5))
  n6 = Constant() {r=1}
  n7 = Desaturation(None=@n5, Fraction=Constant() {r=1})
  StaticSwitchParameter(
   True=Multiply(A=@n6, B=ComponentMask(None=@n7) {r=true})
   False=StaticSwitchParameter(True=ParticleColor().A, False=@n6) {parameterName=UseParticleOpacity?}
  ) {parameterName=MultiplyColorAgainstOpacity}
 ]
}
